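# Animation Blueprint for the M_MED_Prime_Order_NPC body mesh that copies its attached parent's pose and adds a hood correction.
# Each update maps head yaw relative to neck_02, measured from the reference pose, into HeadBack.
# HeadBack weights an additive bone-space translation of dyn_skyDiveHoodFix.
# A bool blend selects that corrected pose while crouching, otherwise the plain copied pose; its crouch-and-accelerating term is redundant.
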AnimBlueprint /BRCosmetics/Characters/Player/Male/Medium/Bodies/M_MED_Prime_Order_NPC/Meshes/M_MED_Prime_Order_NPC_AnimBP
target_skeleton=&/M_MED_Prime_Order_NPC_Skeleton
graphs [
 AnimGraph {nodes [
   n1 = "Transform (Modify) Bone - Bone: dyn_skyDiveHoodFix"(
    ComponentPose=Local To Component(LocalPose=Copy Pose From Mesh() {bUseAttachedParent=true})
    Alpha=Get HeadBack() -> (HeadBack:double)
    Translation="3.5,0,0"
    Rotation="0,0,0"
    Scale="1,1,1"
   ) {
    rotationMode=BMM_Additive
    rotationSpace=BCS_BoneSpace
    scaleMode=BMM_Additive
    scaleSpace=BCS_BoneSpace
    translationMode=BMM_Additive
    translationSpace=BCS_BoneSpace
    alphaBoolBlend {blendOption=Linear}
    boneToModify {boneName=dyn_skyDiveHoodFix}
    translation {x=3.5}
   }
   n2 = AND Boolean(A=Get bIsCrouching(), B=Get bIsAccelerating2D())
   n3 = Get bIsCrouching()
   Output Pose(Result=Blend Poses by bool(
     BlendPose_0=Component To Local(ComponentPose=@n1)
     BlendPose_1=Copy Pose From Mesh() {bUseAttachedParent=true}
     BlendTime_0=1
     BlendTime_1=1
     bActiveValue=OR Boolean(A=@n2, B=@n3)
    ) {blendType=Linear}) {layerGroup=None}
  ]}
 EventGraph {nodes [
   Event BlueprintUpdateAnimation()
   n1 = GetOwnerSkeletalMeshComponent()
   n2 = BreakRotator(InRot=BreakTransform(InTransform=GetDeltaTransformFromRefPose(self=@n1, BoneName=head, BaseName=neck_02)).Rotation)
   > Set HeadBack(HeadBack:double=float * float(A=MapRangeClamped(Value=@n2.Yaw, InRangeA=0, InRangeB=-38), B=1)) -> (Output_Get:double)
  ]}
]
variables [HeadBack {type=number}]
